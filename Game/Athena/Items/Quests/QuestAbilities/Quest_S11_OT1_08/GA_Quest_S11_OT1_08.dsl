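; Quest gameplay ability that tracks how far the player travels while chilled (IcyFeet grenade status) to progress a quest.
; On activation it resets the distance and starts a looping timer that adds movement since the last sampled location to a running total.
; When the total reaches the required meters, it sends the ChilledDistanceTraveled stat event and ends.
; It also stops the timer, takes a final sample and ends when the chilled tag is removed.
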(event Ability|EventActivateAbilityFromEvent
  (params (EventData :type Gameplay%20Event%20Data%20Structure%20%28by%20ref%29))
  (|CallParentFunction EventData)
  (CallFunction|ResetAbility)
  ;[1 1 1] Start tracking loop
  (bind _returnvalue (Utilities|Time|SetTimerbyEvent (delegate Custom|UpdateDistance) (Variables|GAVars|Setup|GetProgressInterval) true))
  (Variables|GAVars|Data|SetTimer_Progress _returnvalue)
  ;[1 1 1] Clean up and end when we are no longer chilled
  (Ability|Tasks|WaitGameplayTagRemove "(TagName=\"Gameplay.Status.IcyFeet.Grenade\")"
    (:Removed
      (Utilities|Time|ClearandInvalidateTimerbyHandle (Variables|GAVars|Data|GetTimer_Progress))
      (CallFunction|UpdateDistance)
      (Ability|EndAbility))))

(event Custom|UpdateDistance
  ;[1 1 1] Calc distance traveled
  (bind _output_get (Variables|GAVars|Data|SetCurrentLocation (Transformation|GetActorLocation (Variables|GAVars|Data|GetOwningPlayerPawn))))
  (bind _distance (Class|MissionBlueprintFunctionLibrary|DistanceBetweenTwoVectors _output_get (Variables|GAVars|Data|GetPreviousLocation)))
  (Variables|GAVars|Data|SetTotalDistance (+ _distance (Variables|GAVars|Data|GetTotalDistance)))
  (Variables|GAVars|Data|SetPreviousLocation (Variables|GAVars|Data|GetCurrentLocation))
  (CallFunction|TryProgressQuest))

(event Custom|TryProgressQuest
  (if (>= (/ (Variables|GAVars|Data|GetTotalDistance) 1e+02) (Variables|GAVars|Setup|GetQuestReqDistanceMeters))
    ;[1 1 1] Send quest update
    (Quest|SendComplexCustomStatEvent (Variables|GAVars|Data|GetOwningPlayerQuestManager) 0 "(GameplayTags=((TagName=\"Athena.Quests.ChilledDistanceTraveled\")))" "(GameplayTags=)" 1)
    (Utilities|Time|ClearandInvalidateTimerbyHandle (Variables|GAVars|Data|GetTimer_Progress))
    (Ability|EndAbility)))

(fn ResetAbility
  (Variables|GAVars|Data|SetPreviousLocation (Transformation|GetActorLocation (Variables|GAVars|Data|GetOwningPlayerPawn)))
  (Variables|GAVars|Data|SetTotalDistance 0.0))
